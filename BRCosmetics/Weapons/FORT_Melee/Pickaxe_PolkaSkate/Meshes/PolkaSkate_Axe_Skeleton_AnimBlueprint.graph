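# Animation Blueprint for the PolkaSkate pickaxe skeleton that adds physics-driven secondary motion to its umbrella fabric, stem, petal and leaf bones.
# Single-bone Anim Dynamics nodes run in series from the cached 'InputDefault' pose, using angular springs, tight limits and gravity overrides.
# The visible graph holds two such chains over the same bones, differently tuned in damping, springs and joint offsets.
# Fabric simulates root-relative in one chain and bone-relative to the umbrella base in the other.
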AnimBlueprint /BRCosmetics/Weapons/FORT_Melee/Pickaxe_PolkaSkate/Meshes/PolkaSkate_Axe_Skeleton_AnimBlueprint
target_skeleton=&/PolkaSkate_Axe_Skeleton
graphs [
 AnimGraph {nodes [
   n1 = Anim Dynamics - Bone: dyn_umbrella_fabric_01!(
    ComponentPose=Local To Component(LocalPose=Use cached pose 'InputDefault'())
    Alpha=1
    SimSpaceSettings:Anim Phys Sim Space Settings Structure=(SimSpaceAngularAlpha=0,MaxAngularVelocity=10000,MaxAngularAcceleration=10000,ExternalAngularVelocity=(X=0,Y=0,Z=0))
    ExternalForce="0,0,0"
   ) -> (Pose) {
    angularDampingOverride=0.99
    angularSpringConstant=600
    bAngularSpring=true
    bGravityOverrideInSimSpace=true
    bOverrideAngularDamping=true
    bUseGravityOverride=true
    linearDampingOverride=1
    simulationSpace=RootRelative
    alphaBoolBlend {blendOption=Linear}
    boundBone {boneName=dyn_umbrella_fabric_01}
    componentAppliedLinearAccClamp=<3000 3000 3000>
    componentLinearAccScale=<0.7 0.7 0.7>
    componentLinearVelScale=<0.7 0.7 0.7>
    gravityOverride {x=-100, z=-450}
    physicsBodyDefinitions [{
      boundBone {boneName=dyn_umbrella_fabric_01}
      boxExtents=<14 14 14>
      constraintSetup {angularLimitsMax=<5 3 3>, angularLimitsMin=<-5 -5 -3>, angularTarget=<1 0 0>}
      localJointOffset=<0.13 0 0>
     }]
    relativeSpaceBone {boneName=def_umbrella_base}
   }
   n2 = Anim Dynamics - Bone: dyn_umbrella_fabric_11!(
    ComponentPose=@n1
    Alpha=1
    SimSpaceSettings:Anim Phys Sim Space Settings Structure=(SimSpaceAngularAlpha=0,MaxAngularVelocity=10000,MaxAngularAcceleration=10000,ExternalAngularVelocity=(X=0,Y=0,Z=0))
    ExternalForce="0,0,0"
   ) -> (Pose) {
    angularDampingOverride=0.99
    angularSpringConstant=600
    bAngularSpring=true
    bGravityOverrideInSimSpace=true
    bOverrideAngularDamping=true
    bUseGravityOverride=true
    linearDampingOverride=1
    simulationSpace=RootRelative
    alphaBoolBlend {blendOption=Linear}
    boundBone {boneName=dyn_umbrella_fabric_11}
    componentAppliedLinearAccClamp=<3000 3000 3000>
    componentLinearAccScale=<0.7 0.7 0.7>
    componentLinearVelScale=<0.7 0.7 0.7>
    gravityOverride {x=-100, z=-450}
    physicsBodyDefinitions [{
      boundBone {boneName=dyn_umbrella_fabric_11}
      boxExtents=<14.5 14.5 14.5>
      constraintSetup {angularLimitsMax=<5 3 3>, angularLimitsMin=<-5 -5 -3>, angularTarget=<1 0 0>}
      localJointOffset=<0.11 0 0>
     }]
    relativeSpaceBone {boneName=def_umbrella_base}
   }
   n3 = Anim Dynamics - Bone: dyn_umbrella_fabric_21!(
    ComponentPose=@n2
    Alpha=1
    SimSpaceSettings:Anim Phys Sim Space Settings Structure=(SimSpaceAngularAlpha=0,MaxAngularVelocity=10000,MaxAngularAcceleration=10000,ExternalAngularVelocity=(X=0,Y=0,Z=0))
    ExternalForce="0,0,0"
   ) -> (Pose) {
    angularDampingOverride=0.99
    angularSpringConstant=600
    bAngularSpring=true
    bGravityOverrideInSimSpace=true
    bOverrideAngularDamping=true
    bUseGravityOverride=true
    linearDampingOverride=1
    simulationSpace=RootRelative
    alphaBoolBlend {blendOption=Linear}
    boundBone {boneName=dyn_umbrella_fabric_21}
    componentAppliedLinearAccClamp=<3000 3000 3000>
    componentLinearAccScale=<0.7 0.7 0.7>
    componentLinearVelScale=<0.7 0.7 0.7>
    gravityOverride {x=-100, z=-450}
    physicsBodyDefinitions [{
      boundBone {boneName=dyn_umbrella_fabric_21}
      boxExtents=<13.5 13.5 13.5>
      constraintSetup {angularLimitsMax=<5 3 3>, angularLimitsMin=<-5 -5 -3>, angularTarget=<1 0 0>}
      localJointOffset=<0.12 0 0>
     }]
    relativeSpaceBone {boneName=def_umbrella_base}
   }
   n4 = Anim Dynamics - Bone: dyn_umbrella_fabric_31!(
    ComponentPose=@n3
    Alpha=1
    SimSpaceSettings:Anim Phys Sim Space Settings Structure=(SimSpaceAngularAlpha=0,MaxAngularVelocity=10000,MaxAngularAcceleration=10000,ExternalAngularVelocity=(X=0,Y=0,Z=0))
    ExternalForce="0,0,0"
   ) -> (Pose) {
    angularDampingOverride=0.99
    angularSpringConstant=600
    bAngularSpring=true
    bGravityOverrideInSimSpace=true
    bOverrideAngularDamping=true
    bUseGravityOverride=true
    linearDampingOverride=1
    simulationSpace=RootRelative
    alphaBoolBlend {blendOption=Linear}
    boundBone {boneName=dyn_umbrella_fabric_31}
    componentAppliedLinearAccClamp=<3000 3000 3000>
    componentLinearAccScale=<0.7 0.7 0.7>
    componentLinearVelScale=<0.7 0.7 0.7>
    gravityOverride {x=-100, z=-450}
    physicsBodyDefinitions [{
      boundBone {boneName=dyn_umbrella_fabric_31}
      boxExtents=<15 15 15>
      constraintSetup {angularLimitsMax=<5 3 3>, angularLimitsMin=<-5 -5 -3>, angularTarget=<1 0 0>}
      localJointOffset=<0.14 0 0>
     }]
    relativeSpaceBone {boneName=def_umbrella_base}
   }
   n5 = Anim Dynamics - Bone: dyn_umbrella_fabric_41!(
    ComponentPose=@n4
    Alpha=1
    SimSpaceSettings:Anim Phys Sim Space Settings Structure=(SimSpaceAngularAlpha=0,MaxAngularVelocity=10000,MaxAngularAcceleration=10000,ExternalAngularVelocity=(X=0,Y=0,Z=0))
    ExternalForce="0,0,0"
   ) -> (Pose) {
    angularDampingOverride=0.99
    angularSpringConstant=600
    bAngularSpring=true
    bGravityOverrideInSimSpace=true
    bOverrideAngularDamping=true
    bUseGravityOverride=true
    linearDampingOverride=1
    simulationSpace=RootRelative
    alphaBoolBlend {blendOption=Linear}
    boundBone {boneName=dyn_umbrella_fabric_41}
    componentAppliedLinearAccClamp=<3000 3000 3000>
    componentLinearAccScale=<0.7 0.7 0.7>
    componentLinearVelScale=<0.7 0.7 0.7>
    gravityOverride {x=-100, z=-450}
    physicsBodyDefinitions [{
      boundBone {boneName=dyn_umbrella_fabric_41}
      boxExtents=<15 15 15>
      constraintSetup {angularLimitsMax=<5 3 3>, angularLimitsMin=<-5 -5 -3>, angularTarget=<1 0 0>}
      localJointOffset=<0.12 0 0>
     }]
    relativeSpaceBone {boneName=def_umbrella_base}
   }
   n6 = Anim Dynamics - Bone: dyn_umbrella_fabric_51!(
    ComponentPose=@n5
    Alpha=1
    SimSpaceSettings:Anim Phys Sim Space Settings Structure=(SimSpaceAngularAlpha=0,MaxAngularVelocity=10000,MaxAngularAcceleration=10000,ExternalAngularVelocity=(X=0,Y=0,Z=0))
    ExternalForce="0,0,0"
   ) -> (Pose) {
    angularDampingOverride=0.99
    angularSpringConstant=600
    bAngularSpring=true
    bGravityOverrideInSimSpace=true
    bOverrideAngularDamping=true
    bUseGravityOverride=true
    linearDampingOverride=1
    simulationSpace=RootRelative
    alphaBoolBlend {blendOption=Linear}
    boundBone {boneName=dyn_umbrella_fabric_51}
    componentAppliedLinearAccClamp=<3000 3000 3000>
    componentLinearAccScale=<0.7 0.7 0.7>
    componentLinearVelScale=<0.7 0.7 0.7>
    gravityOverride {x=-100, z=-450}
    physicsBodyDefinitions [{
      boundBone {boneName=dyn_umbrella_fabric_51}
      boxExtents=<13 13 13>
      constraintSetup {angularLimitsMax=<5 3 3>, angularLimitsMin=<-5 -5 -3>, angularTarget=<1 0 0>}
      localJointOffset=<0.13 0 0>
     }]
    relativeSpaceBone {boneName=def_umbrella_base}
   }
   n7 = Anim Dynamics - Bone: dyn_umbrella_fabric_61!(
    ComponentPose=@n6
    Alpha=1
    SimSpaceSettings:Anim Phys Sim Space Settings Structure=(SimSpaceAngularAlpha=0,MaxAngularVelocity=10000,MaxAngularAcceleration=10000,ExternalAngularVelocity=(X=0,Y=0,Z=0))
    ExternalForce="0,0,0"
   ) -> (Pose) {
    angularDampingOverride=0.99
    angularSpringConstant=600
    bAngularSpring=true
    bGravityOverrideInSimSpace=true
    bOverrideAngularDamping=true
    bUseGravityOverride=true
    linearDampingOverride=1
    simulationSpace=RootRelative
    alphaBoolBlend {blendOption=Linear}
    boundBone {boneName=dyn_umbrella_fabric_61}
    componentAppliedLinearAccClamp=<3000 3000 3000>
    componentLinearAccScale=<0.7 0.7 0.7>
    componentLinearVelScale=<0.7 0.7 0.7>
    gravityOverride {x=-100, z=-450}
    physicsBodyDefinitions [{
      boundBone {boneName=dyn_umbrella_fabric_61}
      boxExtents=<15.5 15.5 15.5>
      constraintSetup {angularLimitsMax=<5 3 3>, angularLimitsMin=<-5 -5 -3>, angularTarget=<1 0 0>}
      localJointOffset=<0.11 0 0>
     }]
    relativeSpaceBone {boneName=def_umbrella_base}
   }
   n8 = Anim Dynamics - Bone: dyn_umbrella_fabric_71!(
    ComponentPose=@n7
    Alpha=1
    SimSpaceSettings:Anim Phys Sim Space Settings Structure=(SimSpaceAngularAlpha=0,MaxAngularVelocity=10000,MaxAngularAcceleration=10000,ExternalAngularVelocity=(X=0,Y=0,Z=0))
    ExternalForce="0,0,0"
   ) -> (Pose) {
    angularDampingOverride=0.99
    angularSpringConstant=600
    bAngularSpring=true
    bGravityOverrideInSimSpace=true
    bOverrideAngularDamping=true
    bUseGravityOverride=true
    linearDampingOverride=1
    simulationSpace=RootRelative
    alphaBoolBlend {blendOption=Linear}
    boundBone {boneName=dyn_umbrella_fabric_71}
    componentAppliedLinearAccClamp=<3000 3000 3000>
    componentLinearAccScale=<0.7 0.7 0.7>
    componentLinearVelScale=<0.7 0.7 0.7>
    gravityOverride {x=-100, z=-450}
    physicsBodyDefinitions [{
      boundBone {boneName=dyn_umbrella_fabric_71}
      boxExtents=<15 15 15>
      constraintSetup {angularLimitsMax=<5 3 3>, angularLimitsMin=<-5 -5 -3>, angularTarget=<1 0 0>}
      localJointOffset=<0.13 0 0>
     }]
    relativeSpaceBone {boneName=def_umbrella_base}
   }
   n9 = Anim Dynamics - Bone: dyn_stem!(
    ComponentPose=@n8
    Alpha=1
    SimSpaceSettings:Anim Phys Sim Space Settings Structure=(SimSpaceAngularAlpha=0,MaxAngularVelocity=10000,MaxAngularAcceleration=10000,ExternalAngularVelocity=(X=0,Y=0,Z=0))
    ExternalForce="0,0,0"
   ) -> (Pose) {
    angularDampingOverride=0.999
    angularSpringConstant=300
    bAngularSpring=true
    bOverrideAngularDamping=true
    bUseGravityOverride=true
    simulationSpace=BoneRelative
    alphaBoolBlend {blendOption=Linear}
    boundBone {boneName=dyn_stem}
    componentAppliedLinearAccClamp=<3000 3000 3000>
    componentLinearAccScale=<0.5 0.5 0.5>
    componentLinearVelScale=<0.5 0.5 0.5>
    gravityOverride {z=-980}
    physicsBodyDefinitions [{
      boundBone {boneName=dyn_stem}
      boxExtents=<11 11 11>
      constraintSetup {angularLimitsMax=<10 0 0>, angularLimitsMin=<-10 0 -10>, angularTarget=<1 0 0>}
      localJointOffset=<-0.1 0 0>
     }]
    relativeSpaceBone {boneName=def_umbrella_base}
   }
   n10 = Anim Dynamics - Bone: dyn_petal_up!(
    ComponentPose=@n9
    Alpha=1
    SimSpaceSettings:Anim Phys Sim Space Settings Structure=(SimSpaceAngularAlpha=0,MaxAngularVelocity=10000,MaxAngularAcceleration=10000,ExternalAngularVelocity=(X=0,Y=0,Z=0))
    ExternalForce="0,0,0"
   ) -> (Pose) {
    angularDampingOverride=0.999
    angularSpringConstant=900
    bAngularSpring=true
    bOverrideAngularDamping=true
    bUseGravityOverride=true
    simulationSpace=BoneRelative
    alphaBoolBlend {blendOption=Linear}
    boundBone {boneName=dyn_petal_up}
    componentAppliedLinearAccClamp=<3000 3000 3000>
    componentLinearAccScale=<2 2 2>
    componentLinearVelScale=<2 2 2>
    gravityOverride {z=-980}
    physicsBodyDefinitions [{
      boundBone {boneName=dyn_petal_up}
      boxExtents=<11 11 11>
      constraintSetup {angularLimitsMax=<0 0 30>, angularLimitsMin=<0 0 -5>, angularTarget=<1 0 0>}
      localJointOffset=<0.2 0 0>
     }]
    relativeSpaceBone {boneName=def_umbrella_base}
   }
   n11 = Anim Dynamics - Bone: dyn_petal_r!(
    ComponentPose=@n10
    Alpha=1
    SimSpaceSettings:Anim Phys Sim Space Settings Structure=(SimSpaceAngularAlpha=0,MaxAngularVelocity=10000,MaxAngularAcceleration=10000,ExternalAngularVelocity=(X=0,Y=0,Z=0))
    ExternalForce="0,0,0"
   ) -> (Pose) {
    angularDampingOverride=0.999
    angularSpringConstant=900
    bAngularSpring=true
    bOverrideAngularDamping=true
    bUseGravityOverride=true
    simulationSpace=BoneRelative
    alphaBoolBlend {blendOption=Linear}
    boundBone {boneName=dyn_petal_r}
    componentAppliedLinearAccClamp=<3000 3000 3000>
    componentLinearAccScale=<2 2 2>
    componentLinearVelScale=<2 2 2>
    gravityOverride {z=-980}
    physicsBodyDefinitions [{
      boundBone {boneName=dyn_petal_r}
      boxExtents=<11 11 11>
      constraintSetup {angularLimitsMax=<0 0 45>, angularLimitsMin=<0 0 -5>, angularTarget=<1 0 0>}
      localJointOffset=<0.2 0 0>
     }]
    relativeSpaceBone {boneName=def_umbrella_base}
   }
   n12 = Anim Dynamics - Bone: dyn_petal_down!(
    ComponentPose=@n11
    Alpha=1
    SimSpaceSettings:Anim Phys Sim Space Settings Structure=(SimSpaceAngularAlpha=0,MaxAngularVelocity=10000,MaxAngularAcceleration=10000,ExternalAngularVelocity=(X=0,Y=0,Z=0))
    ExternalForce="0,0,0"
   ) -> (Pose) {
    angularDampingOverride=0.999
    angularSpringConstant=900
    bAngularSpring=true
    bOverrideAngularDamping=true
    bUseGravityOverride=true
    simulationSpace=BoneRelative
    alphaBoolBlend {blendOption=Linear}
    boundBone {boneName=dyn_petal_down}
    componentAppliedLinearAccClamp=<3000 3000 3000>
    componentLinearAccScale=<2 2 2>
    componentLinearVelScale=<2 2 2>
    gravityOverride {z=-980}
    physicsBodyDefinitions [{
      boundBone {boneName=dyn_petal_down}
      boxExtents=<11 11 11>
      constraintSetup {angularLimitsMax=<0 0 5>, angularLimitsMin=<0 0 -30>, angularTarget=<1 0 0>}
      localJointOffset=<0.2 0 0>
     }]
    relativeSpaceBone {boneName=def_umbrella_base}
   }
   n13 = Anim Dynamics - Bone: dyn_petal_left!(
    ComponentPose=@n12
    Alpha=1
    SimSpaceSettings:Anim Phys Sim Space Settings Structure=(SimSpaceAngularAlpha=0,MaxAngularVelocity=10000,MaxAngularAcceleration=10000,ExternalAngularVelocity=(X=0,Y=0,Z=0))
    ExternalForce="0,0,0"
   ) -> (Pose) {
    angularDampingOverride=0.999
    angularSpringConstant=900
    bAngularSpring=true
    bOverrideAngularDamping=true
    bUseGravityOverride=true
    simulationSpace=BoneRelative
    alphaBoolBlend {blendOption=Linear}
    boundBone {boneName=dyn_petal_left}
    componentAppliedLinearAccClamp=<3000 3000 3000>
    componentLinearAccScale=<2 2 2>
    componentLinearVelScale=<2 2 2>
    gravityOverride {z=-980}
    physicsBodyDefinitions [{
      boundBone {boneName=dyn_petal_left}
      boxExtents=<11 11 11>
      constraintSetup {angularLimitsMax=<0 0 5>, angularLimitsMin=<0 0 -30>, angularTarget=<1 0 0>}
      localJointOffset=<0.2 0 0>
     }]
    relativeSpaceBone {boneName=def_umbrella_base}
   }
   n14 = Anim Dynamics - Bone: dyn_umbrella_up!(
    ComponentPose=@n13
    Alpha=1
    SimSpaceSettings:Anim Phys Sim Space Settings Structure=(SimSpaceAngularAlpha=0,MaxAngularVelocity=10000,MaxAngularAcceleration=10000,ExternalAngularVelocity=(X=0,Y=0,Z=0))
    ExternalForce="0,0,0"
   ) -> (Pose) {
    angularDampingOverride=0.999
    angularSpringConstant=300
    bAngularSpring=true
    bOverrideAngularDamping=true
    bOverrideLinearDamping=true
    bUseGravityOverride=true
    linearDampingOverride=0.9999
    simulationSpace=BoneRelative
    alphaBoolBlend {blendOption=Linear}
    boundBone {boneName=dyn_umbrella_up}
    componentAppliedLinearAccClamp=<3000 3000 3000>
    componentLinearAccScale=<1 1 1>
    componentLinearVelScale=<1 1 1>
    gravityOverride {z=-980}
    physicsBodyDefinitions [{
      boundBone {boneName=dyn_umbrella_up}
      boxExtents=<12 12 12>
      constraintSetup {angularLimitsMax=<0 0 10>, angularLimitsMin=<0 0 -10>, angularTarget=<1 0 0>}
      localJointOffset=<0.1 0 0>
     }]
    relativeSpaceBone {boneName=def_umbrella_base}
   }
   n15 = Anim Dynamics - Bone: dyn_leave_01_l!(
    ComponentPose=@n14
    Alpha=1
    SimSpaceSettings:Anim Phys Sim Space Settings Structure=(SimSpaceAngularAlpha=0,MaxAngularVelocity=10000,MaxAngularAcceleration=10000,ExternalAngularVelocity=(X=0,Y=0,Z=0))
    ExternalForce="0,0,0"
   ) -> (Pose) {
    angularDampingOverride=0.999
    angularSpringConstant=900
    bAngularSpring=true
    bOverrideAngularDamping=true
    bUseGravityOverride=true
    simulationSpace=BoneRelative
    alphaBoolBlend {blendOption=Linear}
    boundBone {boneName=dyn_leave_01_l}
    componentAppliedLinearAccClamp=<3000 3000 3000>
    componentLinearAccScale=<1 1 1>
    componentLinearVelScale=<1 1 1>
    gravityOverride {z=-980}
    physicsBodyDefinitions [{
      boundBone {boneName=dyn_leave_01_l}
      boxExtents=<11 11 11>
      constraintSetup {angularLimitsMax=<0 0 5>, angularLimitsMin=<0 0 -20>, angularTarget=<1 0 0>}
      localJointOffset=<0.2 0 0>
     }]
    relativeSpaceBone {boneName=def_umbrella_base}
   }
   n16 = Anim Dynamics - Bone: dyn_leave_02_l!(
    ComponentPose=@n15
    Alpha=1
    SimSpaceSettings:Anim Phys Sim Space Settings Structure=(SimSpaceAngularAlpha=0,MaxAngularVelocity=10000,MaxAngularAcceleration=10000,ExternalAngularVelocity=(X=0,Y=0,Z=0))
    ExternalForce="0,0,0"
   ) -> (Pose) {
    angularDampingOverride=0.999
    angularSpringConstant=900
    bAngularSpring=true
    bOverrideAngularDamping=true
    bUseGravityOverride=true
    simulationSpace=BoneRelative
    alphaBoolBlend {blendOption=Linear}
    boundBone {boneName=dyn_leave_02_l}
    componentAppliedLinearAccClamp=<3000 3000 3000>
    componentLinearAccScale=<1 1 1>
    componentLinearVelScale=<1 1 1>
    gravityOverride {z=-980}
    physicsBodyDefinitions [{
      boundBone {boneName=dyn_leave_02_l}
      boxExtents=<11 11 11>
      constraintSetup {angularLimitsMax=<0 0 5>, angularLimitsMin=<0 0 -80>, angularTarget=<1 0 0>}
      localJointOffset=<0.2 0 0>
     }]
    relativeSpaceBone {boneName=def_umbrella_base}
   }
   n17 = Anim Dynamics - Bone: dyn_leave_01_r!(
    ComponentPose=@n16
    Alpha=1
    SimSpaceSettings:Anim Phys Sim Space Settings Structure=(SimSpaceAngularAlpha=0,MaxAngularVelocity=10000,MaxAngularAcceleration=10000,ExternalAngularVelocity=(X=0,Y=0,Z=0))
    ExternalForce="0,0,0"
   ) -> (Pose) {
    angularDampingOverride=0.999
    angularSpringConstant=900
    bAngularSpring=true
    bOverrideAngularDamping=true
    bUseGravityOverride=true
    simulationSpace=BoneRelative
    alphaBoolBlend {blendOption=Linear}
    boundBone {boneName=dyn_leave_01_r}
    componentAppliedLinearAccClamp=<3000 3000 3000>
    componentLinearAccScale=<1 1 1>
    componentLinearVelScale=<1 1 1>
    gravityOverride {z=-980}
    physicsBodyDefinitions [{
      boundBone {boneName=dyn_leave_01_r}
      boxExtents=<11 11 11>
      constraintSetup {angularLimitsMax=<0 0 5>, angularLimitsMin=<0 0 -20>, angularTarget=<1 0 0>}
      localJointOffset=<0.2 0 0>
     }]
    relativeSpaceBone {boneName=def_umbrella_base}
   }
   n18 = Anim Dynamics - Bone: dyn_leave_02_r!(
    ComponentPose=@n17
    Alpha=1
    SimSpaceSettings:Anim Phys Sim Space Settings Structure=(SimSpaceAngularAlpha=0,MaxAngularVelocity=10000,MaxAngularAcceleration=10000,ExternalAngularVelocity=(X=0,Y=0,Z=0))
    ExternalForce="0,0,0"
   ) -> (Pose) {
    angularDampingOverride=0.999
    angularSpringConstant=900
    bAngularSpring=true
    bOverrideAngularDamping=true
    bUseGravityOverride=true
    simulationSpace=BoneRelative
    alphaBoolBlend {blendOption=Linear}
    boundBone {boneName=dyn_leave_02_r}
    componentAppliedLinearAccClamp=<3000 3000 3000>
    componentLinearAccScale=<1 1 1>
    componentLinearVelScale=<1 1 1>
    gravityOverride {z=-980}
    physicsBodyDefinitions [{
      boundBone {boneName=dyn_leave_02_r}
      boxExtents=<11 11 11>
      constraintSetup {angularLimitsMax=<0 0 5>, angularLimitsMin=<0 0 -80>, angularTarget=<1 0 0>}
      localJointOffset=<0.2 0 0>
     }]
    relativeSpaceBone {boneName=def_umbrella_base}
   }
   n19 = Anim Dynamics - Bone: dyn_stem_low_01!(
    ComponentPose=@n18
    Alpha=1
    SimSpaceSettings:Anim Phys Sim Space Settings Structure=(SimSpaceAngularAlpha=0,MaxAngularVelocity=10000,MaxAngularAcceleration=10000,ExternalAngularVelocity=(X=0,Y=0,Z=0))
    ExternalForce="0,0,0"
   ) -> (Pose) {
    angularDampingOverride=0.99
    angularSpringConstant=600
    bAngularSpring=true
    bOverrideAngularDamping=true
    bUseGravityOverride=true
    simulationSpace=BoneRelative
    alphaBoolBlend {blendOption=Linear}
    boundBone {boneName=dyn_stem_low_01}
    componentAppliedLinearAccClamp=<3000 3000 3000>
    componentLinearAccScale=<1 1 1>
    componentLinearVelScale=<1 1 1>
    gravityOverride {z=-980}
    physicsBodyDefinitions [{
      boundBone {boneName=dyn_stem_low_01}
      boxExtents=<11 11 11>
      constraintSetup {angularLimitsMax=<0 5 20>, angularLimitsMin=<0 -5 -60>, angularTarget=<1 0 0>}
      localJointOffset=<0.1 0 0>
     }]
    relativeSpaceBone {boneName=def_umbrella_base}
   }
   n20 = Anim Dynamics - Bone: dyn_stem_low_02!(
    ComponentPose=@n19
    Alpha=1
    SimSpaceSettings:Anim Phys Sim Space Settings Structure=(SimSpaceAngularAlpha=0,MaxAngularVelocity=10000,MaxAngularAcceleration=10000,ExternalAngularVelocity=(X=0,Y=0,Z=0))
    ExternalForce="0,0,0"
   ) -> (Pose) {
    angularDampingOverride=0.99
    angularSpringConstant=600
    bAngularSpring=true
    bOverrideAngularDamping=true
    bUseGravityOverride=true
    simulationSpace=BoneRelative
    alphaBoolBlend {blendOption=Linear}
    boundBone {boneName=dyn_stem_low_02}
    componentAppliedLinearAccClamp=<3000 3000 3000>
    componentLinearAccScale=<1 1 1>
    componentLinearVelScale=<1 1 1>
    gravityOverride {z=-980}
    physicsBodyDefinitions [{
      boundBone {boneName=dyn_stem_low_02}
      boxExtents=<11 11 11>
      constraintSetup {angularLimitsMax=<0 5 30>, angularLimitsMin=<0 -5 -30>, angularTarget=<1 0 0>}
      localJointOffset=<0.1 0 0>
     }]
    relativeSpaceBone {boneName=def_umbrella_base}
   }
   n21 = Anim Dynamics - Bone: dyn_umbrella_fabric_01(
    ComponentPose=Local To Component(LocalPose=Use cached pose 'InputDefault'())
    Alpha=1
    SimSpaceSettings=(SimSpaceAngularAlpha=0,MaxAngularVelocity=10000,MaxAngularAcceleration=10000,ExternalAngularVelocity=(X=0,Y=0,Z=0))
    ExternalForce="0,0,0"
   ) {
    angularDampingOverride=0.999
    angularSpringConstant=600
    bAngularSpring=true
    bGravityOverrideInSimSpace=true
    bOverrideAngularDamping=true
    bUseGravityOverride=true
    linearDampingOverride=1
    simulationSpace=BoneRelative
    alphaBoolBlend {blendOption=Linear}
    boundBone {boneName=dyn_umbrella_fabric_01}
    componentAppliedLinearAccClamp=<3000 3000 3000>
    componentLinearAccScale=<0.7 0.7 0.7>
    componentLinearVelScale=<0.7 0.7 0.7>
    gravityOverride {x=-980, z=-300}
    physicsBodyDefinitions [{
      boundBone {boneName=dyn_umbrella_fabric_01}
      boxExtents=<14 14 14>
      constraintSetup {angularLimitsMax=<5 3 3>, angularLimitsMin=<-5 -5 -3>, angularTarget=<1 0 0>}
      localJointOffset=<0.73 0 0>
     }]
    relativeSpaceBone {boneName=def_umbrella_base}
   }
   n22 = Anim Dynamics - Bone: dyn_umbrella_fabric_11(
    ComponentPose=@n21
    Alpha=1
    SimSpaceSettings=(SimSpaceAngularAlpha=0,MaxAngularVelocity=10000,MaxAngularAcceleration=10000,ExternalAngularVelocity=(X=0,Y=0,Z=0))
    ExternalForce="0,0,0"
   ) {
    angularDampingOverride=0.999
    angularSpringConstant=600
    bAngularSpring=true
    bGravityOverrideInSimSpace=true
    bOverrideAngularDamping=true
    bUseGravityOverride=true
    linearDampingOverride=1
    simulationSpace=BoneRelative
    alphaBoolBlend {blendOption=Linear}
    boundBone {boneName=dyn_umbrella_fabric_11}
    componentAppliedLinearAccClamp=<3000 3000 3000>
    componentLinearAccScale=<0.7 0.7 0.7>
    componentLinearVelScale=<0.7 0.7 0.7>
    gravityOverride {x=-980, z=-300}
    physicsBodyDefinitions [{
      boundBone {boneName=dyn_umbrella_fabric_11}
      boxExtents=<14.5 14.5 14.5>
      constraintSetup {angularLimitsMax=<5 3 3>, angularLimitsMin=<-5 -5 -3>, angularTarget=<1 0 0>}
      localJointOffset=<0.71 0 0>
     }]
    relativeSpaceBone {boneName=def_umbrella_base}
   }
   n23 = Anim Dynamics - Bone: dyn_umbrella_fabric_21(
    ComponentPose=@n22
    Alpha=1
    SimSpaceSettings=(SimSpaceAngularAlpha=0,MaxAngularVelocity=10000,MaxAngularAcceleration=10000,ExternalAngularVelocity=(X=0,Y=0,Z=0))
    ExternalForce="0,0,0"
   ) {
    angularDampingOverride=0.999
    angularSpringConstant=600
    bAngularSpring=true
    bGravityOverrideInSimSpace=true
    bOverrideAngularDamping=true
    bUseGravityOverride=true
    linearDampingOverride=1
    simulationSpace=BoneRelative
    alphaBoolBlend {blendOption=Linear}
    boundBone {boneName=dyn_umbrella_fabric_21}
    componentAppliedLinearAccClamp=<3000 3000 3000>
    componentLinearAccScale=<0.7 0.7 0.7>
    componentLinearVelScale=<0.7 0.7 0.7>
    gravityOverride {x=-980, z=-300}
    physicsBodyDefinitions [{
      boundBone {boneName=dyn_umbrella_fabric_21}
      boxExtents=<13.5 13.5 13.5>
      constraintSetup {angularLimitsMax=<5 3 3>, angularLimitsMin=<-5 -5 -3>, angularTarget=<1 0 0>}
      localJointOffset=<0.72 0 0>
     }]
    relativeSpaceBone {boneName=def_umbrella_base}
   }
   n24 = Anim Dynamics - Bone: dyn_umbrella_fabric_31(
    ComponentPose=@n23
    Alpha=1
    SimSpaceSettings=(SimSpaceAngularAlpha=0,MaxAngularVelocity=10000,MaxAngularAcceleration=10000,ExternalAngularVelocity=(X=0,Y=0,Z=0))
    ExternalForce="0,0,0"
   ) {
    angularDampingOverride=0.999
    angularSpringConstant=600
    bAngularSpring=true
    bGravityOverrideInSimSpace=true
    bOverrideAngularDamping=true
    bUseGravityOverride=true
    linearDampingOverride=1
    simulationSpace=BoneRelative
    alphaBoolBlend {blendOption=Linear}
    boundBone {boneName=dyn_umbrella_fabric_31}
    componentAppliedLinearAccClamp=<3000 3000 3000>
    componentLinearAccScale=<0.7 0.7 0.7>
    componentLinearVelScale=<0.7 0.7 0.7>
    gravityOverride {x=-980, z=-300}
    physicsBodyDefinitions [{
      boundBone {boneName=dyn_umbrella_fabric_31}
      boxExtents=<15 15 15>
      constraintSetup {angularLimitsMax=<5 3 3>, angularLimitsMin=<-5 -5 -3>, angularTarget=<1 0 0>}
      localJointOffset=<0.74 0 0>
     }]
    relativeSpaceBone {boneName=def_umbrella_base}
   }
   n25 = Anim Dynamics - Bone: dyn_umbrella_fabric_41(
    ComponentPose=@n24
    Alpha=1
    SimSpaceSettings=(SimSpaceAngularAlpha=0,MaxAngularVelocity=10000,MaxAngularAcceleration=10000,ExternalAngularVelocity=(X=0,Y=0,Z=0))
    ExternalForce="0,0,0"
   ) {
    angularDampingOverride=0.999
    angularSpringConstant=600
    bAngularSpring=true
    bGravityOverrideInSimSpace=true
    bOverrideAngularDamping=true
    bUseGravityOverride=true
    linearDampingOverride=1
    simulationSpace=BoneRelative
    alphaBoolBlend {blendOption=Linear}
    boundBone {boneName=dyn_umbrella_fabric_41}
    componentAppliedLinearAccClamp=<3000 3000 3000>
    componentLinearAccScale=<0.7 0.7 0.7>
    componentLinearVelScale=<0.7 0.7 0.7>
    gravityOverride {x=-980, z=-300}
    physicsBodyDefinitions [{
      boundBone {boneName=dyn_umbrella_fabric_41}
      boxExtents=<15 15 15>
      constraintSetup {angularLimitsMax=<5 3 3>, angularLimitsMin=<-5 -5 -3>, angularTarget=<1 0 0>}
      localJointOffset=<0.72 0 0>
     }]
    relativeSpaceBone {boneName=def_umbrella_base}
   }
   n26 = Anim Dynamics - Bone: dyn_umbrella_fabric_51(
    ComponentPose=@n25
    Alpha=1
    SimSpaceSettings=(SimSpaceAngularAlpha=0,MaxAngularVelocity=10000,MaxAngularAcceleration=10000,ExternalAngularVelocity=(X=0,Y=0,Z=0))
    ExternalForce="0,0,0"
   ) {
    angularDampingOverride=0.999
    angularSpringConstant=600
    bAngularSpring=true
    bGravityOverrideInSimSpace=true
    bOverrideAngularDamping=true
    bUseGravityOverride=true
    linearDampingOverride=1
    simulationSpace=BoneRelative
    alphaBoolBlend {blendOption=Linear}
    boundBone {boneName=dyn_umbrella_fabric_51}
    componentAppliedLinearAccClamp=<3000 3000 3000>
    componentLinearAccScale=<0.7 0.7 0.7>
    componentLinearVelScale=<0.7 0.7 0.7>
    gravityOverride {x=-980, z=-300}
    physicsBodyDefinitions [{
      boundBone {boneName=dyn_umbrella_fabric_51}
      boxExtents=<13 13 13>
      constraintSetup {angularLimitsMax=<5 3 3>, angularLimitsMin=<-5 -5 -3>, angularTarget=<1 0 0>}
      localJointOffset=<0.73 0 0>
     }]
    relativeSpaceBone {boneName=def_umbrella_base}
   }
   n27 = Anim Dynamics - Bone: dyn_umbrella_fabric_61(
    ComponentPose=@n26
    Alpha=1
    SimSpaceSettings=(SimSpaceAngularAlpha=0,MaxAngularVelocity=10000,MaxAngularAcceleration=10000,ExternalAngularVelocity=(X=0,Y=0,Z=0))
    ExternalForce="0,0,0"
   ) {
    angularDampingOverride=0.999
    angularSpringConstant=600
    bAngularSpring=true
    bGravityOverrideInSimSpace=true
    bOverrideAngularDamping=true
    bUseGravityOverride=true
    linearDampingOverride=1
    simulationSpace=BoneRelative
    alphaBoolBlend {blendOption=Linear}
    boundBone {boneName=dyn_umbrella_fabric_61}
    componentAppliedLinearAccClamp=<3000 3000 3000>
    componentLinearAccScale=<0.7 0.7 0.7>
    componentLinearVelScale=<0.7 0.7 0.7>
    gravityOverride {x=-980, z=-300}
    physicsBodyDefinitions [{
      boundBone {boneName=dyn_umbrella_fabric_61}
      boxExtents=<15.5 15.5 15.5>
      constraintSetup {angularLimitsMax=<5 3 3>, angularLimitsMin=<-5 -5 -3>, angularTarget=<1 0 0>}
      localJointOffset=<0.71 0 0>
     }]
    relativeSpaceBone {boneName=def_umbrella_base}
   }
   n28 = Anim Dynamics - Bone: dyn_umbrella_fabric_71(
    ComponentPose=@n27
    Alpha=1
    SimSpaceSettings=(SimSpaceAngularAlpha=0,MaxAngularVelocity=10000,MaxAngularAcceleration=10000,ExternalAngularVelocity=(X=0,Y=0,Z=0))
    ExternalForce="0,0,0"
   ) {
    angularDampingOverride=0.999
    angularSpringConstant=600
    bAngularSpring=true
    bGravityOverrideInSimSpace=true
    bOverrideAngularDamping=true
    bUseGravityOverride=true
    linearDampingOverride=1
    simulationSpace=BoneRelative
    alphaBoolBlend {blendOption=Linear}
    boundBone {boneName=dyn_umbrella_fabric_71}
    componentAppliedLinearAccClamp=<3000 3000 3000>
    componentLinearAccScale=<0.7 0.7 0.7>
    componentLinearVelScale=<0.7 0.7 0.7>
    gravityOverride {x=-980, z=-300}
    physicsBodyDefinitions [{
      boundBone {boneName=dyn_umbrella_fabric_71}
      boxExtents=<15 15 15>
      constraintSetup {angularLimitsMax=<5 3 3>, angularLimitsMin=<-5 -5 -3>, angularTarget=<1 0 0>}
      localJointOffset=<0.73 0 0>
     }]
    relativeSpaceBone {boneName=def_umbrella_base}
   }
   n29 = Anim Dynamics - Bone: dyn_stem(
    ComponentPose=@n28
    Alpha=1
    SimSpaceSettings=(SimSpaceAngularAlpha=0,MaxAngularVelocity=10000,MaxAngularAcceleration=10000,ExternalAngularVelocity=(X=0,Y=0,Z=0))
    ExternalForce="0,0,0"
   ) {
    angularDampingOverride=0.99
    angularSpringConstant=300
    bAngularSpring=true
    bOverrideAngularDamping=true
    bUseGravityOverride=true
    simulationSpace=BoneRelative
    alphaBoolBlend {blendOption=Linear}
    boundBone {boneName=dyn_stem}
    componentAppliedLinearAccClamp=<3000 3000 3000>
    componentLinearAccScale=<0.5 0.5 0.5>
    componentLinearVelScale=<0.5 0.5 0.5>
    gravityOverride {z=-980}
    physicsBodyDefinitions [{
      boundBone {boneName=dyn_stem}
      boxExtents=<11 11 11>
      constraintSetup {angularLimitsMax=<10 0 0>, angularLimitsMin=<-10 0 -10>, angularTarget=<1 0 0>}
      localJointOffset=<-0.7 0 0>
     }]
    relativeSpaceBone {boneName=def_umbrella_base}
   }
   n30 = Anim Dynamics - Bone: dyn_petal_up(
    ComponentPose=@n29
    Alpha=1
    SimSpaceSettings=(SimSpaceAngularAlpha=0,MaxAngularVelocity=10000,MaxAngularAcceleration=10000,ExternalAngularVelocity=(X=0,Y=0,Z=0))
    ExternalForce="0,0,0"
   ) {
    angularDampingOverride=0.99
    angularSpringConstant=600
    bAngularSpring=true
    bOverrideAngularDamping=true
    bUseGravityOverride=true
    simulationSpace=BoneRelative
    alphaBoolBlend {blendOption=Linear}
    boundBone {boneName=dyn_petal_up}
    componentAppliedLinearAccClamp=<3000 3000 3000>
    componentLinearAccScale=<2 2 2>
    componentLinearVelScale=<2 2 2>
    gravityOverride {z=-980}
    physicsBodyDefinitions [{
      boundBone {boneName=dyn_petal_up}
      boxExtents=<11 11 11>
      constraintSetup {angularLimitsMax=<0 0 30>, angularLimitsMin=<0 0 -5>, angularTarget=<1 0 0>}
      localJointOffset=<0.3 0 0>
     }]
    relativeSpaceBone {boneName=def_umbrella_base}
   }
   n31 = Anim Dynamics - Bone: dyn_petal_r(
    ComponentPose=@n30
    Alpha=1
    SimSpaceSettings=(SimSpaceAngularAlpha=0,MaxAngularVelocity=10000,MaxAngularAcceleration=10000,ExternalAngularVelocity=(X=0,Y=0,Z=0))
    ExternalForce="0,0,0"
   ) {
    angularDampingOverride=0.99
    angularSpringConstant=600
    bAngularSpring=true
    bOverrideAngularDamping=true
    bUseGravityOverride=true
    simulationSpace=BoneRelative
    alphaBoolBlend {blendOption=Linear}
    boundBone {boneName=dyn_petal_r}
    componentAppliedLinearAccClamp=<3000 3000 3000>
    componentLinearAccScale=<2 2 2>
    componentLinearVelScale=<2 2 2>
    gravityOverride {z=-980}
    physicsBodyDefinitions [{
      boundBone {boneName=dyn_petal_r}
      boxExtents=<11 11 11>
      constraintSetup {angularLimitsMax=<0 0 45>, angularLimitsMin=<0 0 -5>, angularTarget=<1 0 0>}
      localJointOffset=<0.3 0 0>
     }]
    relativeSpaceBone {boneName=def_umbrella_base}
   }
   n32 = Anim Dynamics - Bone: dyn_petal_down(
    ComponentPose=@n31
    Alpha=1
    SimSpaceSettings=(SimSpaceAngularAlpha=0,MaxAngularVelocity=10000,MaxAngularAcceleration=10000,ExternalAngularVelocity=(X=0,Y=0,Z=0))
    ExternalForce="0,0,0"
   ) {
    angularDampingOverride=0.99
    angularSpringConstant=600
    bAngularSpring=true
    bOverrideAngularDamping=true
    bUseGravityOverride=true
    simulationSpace=BoneRelative
    alphaBoolBlend {blendOption=Linear}
    boundBone {boneName=dyn_petal_down}
    componentAppliedLinearAccClamp=<3000 3000 3000>
    componentLinearAccScale=<2 2 2>
    componentLinearVelScale=<2 2 2>
    gravityOverride {z=-980}
    physicsBodyDefinitions [{
      boundBone {boneName=dyn_petal_down}
      boxExtents=<11 11 11>
      constraintSetup {angularLimitsMax=<0 0 5>, angularLimitsMin=<0 0 -30>, angularTarget=<1 0 0>}
      localJointOffset=<0.3 0 0>
     }]
    relativeSpaceBone {boneName=def_umbrella_base}
   }
   n33 = Anim Dynamics - Bone: dyn_petal_left(
    ComponentPose=@n32
    Alpha=1
    SimSpaceSettings=(SimSpaceAngularAlpha=0,MaxAngularVelocity=10000,MaxAngularAcceleration=10000,ExternalAngularVelocity=(X=0,Y=0,Z=0))
    ExternalForce="0,0,0"
   ) {
    angularDampingOverride=0.99
    angularSpringConstant=600
    bAngularSpring=true
    bOverrideAngularDamping=true
    bUseGravityOverride=true
    simulationSpace=BoneRelative
    alphaBoolBlend {blendOption=Linear}
    boundBone {boneName=dyn_petal_left}
    componentAppliedLinearAccClamp=<3000 3000 3000>
    componentLinearAccScale=<2 2 2>
    componentLinearVelScale=<2 2 2>
    gravityOverride {z=-980}
    physicsBodyDefinitions [{
      boundBone {boneName=dyn_petal_left}
      boxExtents=<11 11 11>
      constraintSetup {angularLimitsMax=<0 0 5>, angularLimitsMin=<0 0 -30>, angularTarget=<1 0 0>}
      localJointOffset=<0.3 0 0>
     }]
    relativeSpaceBone {boneName=def_umbrella_base}
   }
   n34 = Anim Dynamics - Bone: dyn_umbrella_up(
    ComponentPose=@n33
    Alpha=1
    SimSpaceSettings=(SimSpaceAngularAlpha=0,MaxAngularVelocity=10000,MaxAngularAcceleration=10000,ExternalAngularVelocity=(X=0,Y=0,Z=0))
    ExternalForce="0,0,0"
   ) {
    angularDampingOverride=0.999
    angularSpringConstant=600
    bAngularSpring=true
    bOverrideAngularDamping=true
    bOverrideLinearDamping=true
    bUseGravityOverride=true
    linearDampingOverride=0.9999
    simulationSpace=BoneRelative
    alphaBoolBlend {blendOption=Linear}
    boundBone {boneName=dyn_umbrella_up}
    componentAppliedLinearAccClamp=<3000 3000 3000>
    componentLinearAccScale=<1 1 1>
    componentLinearVelScale=<1 1 1>
    gravityOverride {z=-980}
    physicsBodyDefinitions [{
      boundBone {boneName=dyn_umbrella_up}
      boxExtents=<12 12 12>
      constraintSetup {angularLimitsMax=<0 0 10>, angularLimitsMin=<0 0 -10>, angularTarget=<1 0 0>}
      localJointOffset=<0.7 0 0>
     }]
    relativeSpaceBone {boneName=def_umbrella_base}
   }
   n35 = Anim Dynamics - Bone: dyn_leave_01_l(
    ComponentPose=@n34
    Alpha=1
    SimSpaceSettings=(SimSpaceAngularAlpha=0,MaxAngularVelocity=10000,MaxAngularAcceleration=10000,ExternalAngularVelocity=(X=0,Y=0,Z=0))
    ExternalForce="0,0,0"
   ) {
    angularDampingOverride=0.99
    angularSpringConstant=600
    bAngularSpring=true
    bOverrideAngularDamping=true
    bUseGravityOverride=true
    simulationSpace=BoneRelative
    alphaBoolBlend {blendOption=Linear}
    boundBone {boneName=dyn_leave_01_l}
    componentAppliedLinearAccClamp=<3000 3000 3000>
    componentLinearAccScale=<1 1 1>
    componentLinearVelScale=<1 1 1>
    gravityOverride {z=-980}
    physicsBodyDefinitions [{
      boundBone {boneName=dyn_leave_01_l}
      boxExtents=<11 11 11>
      constraintSetup {angularLimitsMax=<0 0 5>, angularLimitsMin=<0 0 -20>, angularTarget=<1 0 0>}
      localJointOffset=<0.8 0 0>
     }]
    relativeSpaceBone {boneName=def_umbrella_base}
   }
   n36 = Anim Dynamics - Bone: dyn_leave_02_l(
    ComponentPose=@n35
    Alpha=1
    SimSpaceSettings=(SimSpaceAngularAlpha=0,MaxAngularVelocity=10000,MaxAngularAcceleration=10000,ExternalAngularVelocity=(X=0,Y=0,Z=0))
    ExternalForce="0,0,0"
   ) {
    angularDampingOverride=0.99
    angularSpringConstant=600
    bAngularSpring=true
    bOverrideAngularDamping=true
    bUseGravityOverride=true
    simulationSpace=BoneRelative
    alphaBoolBlend {blendOption=Linear}
    boundBone {boneName=dyn_leave_02_l}
    componentAppliedLinearAccClamp=<3000 3000 3000>
    componentLinearAccScale=<1 1 1>
    componentLinearVelScale=<1 1 1>
    gravityOverride {z=-980}
    physicsBodyDefinitions [{
      boundBone {boneName=dyn_leave_02_l}
      boxExtents=<11 11 11>
      constraintSetup {angularLimitsMax=<0 0 5>, angularLimitsMin=<0 0 -80>, angularTarget=<1 0 0>}
      localJointOffset=<0.8 0 0>
     }]
    relativeSpaceBone {boneName=def_umbrella_base}
   }
   n37 = Anim Dynamics - Bone: dyn_leave_01_r(
    ComponentPose=@n36
    Alpha=1
    SimSpaceSettings=(SimSpaceAngularAlpha=0,MaxAngularVelocity=10000,MaxAngularAcceleration=10000,ExternalAngularVelocity=(X=0,Y=0,Z=0))
    ExternalForce="0,0,0"
   ) {
    angularDampingOverride=0.99
    angularSpringConstant=600
    bAngularSpring=true
    bOverrideAngularDamping=true
    bUseGravityOverride=true
    simulationSpace=BoneRelative
    alphaBoolBlend {blendOption=Linear}
    boundBone {boneName=dyn_leave_01_r}
    componentAppliedLinearAccClamp=<3000 3000 3000>
    componentLinearAccScale=<1 1 1>
    componentLinearVelScale=<1 1 1>
    gravityOverride {z=-980}
    physicsBodyDefinitions [{
      boundBone {boneName=dyn_leave_01_r}
      boxExtents=<11 11 11>
      constraintSetup {angularLimitsMax=<0 0 5>, angularLimitsMin=<0 0 -20>, angularTarget=<1 0 0>}
      localJointOffset=<0.8 0 0>
     }]
    relativeSpaceBone {boneName=def_umbrella_base}
   }
   n38 = Anim Dynamics - Bone: dyn_leave_02_r(
    ComponentPose=@n37
    Alpha=1
    SimSpaceSettings=(SimSpaceAngularAlpha=0,MaxAngularVelocity=10000,MaxAngularAcceleration=10000,ExternalAngularVelocity=(X=0,Y=0,Z=0))
    ExternalForce="0,0,0"
   ) {
    angularDampingOverride=0.99
    angularSpringConstant=600
    bAngularSpring=true
    bOverrideAngularDamping=true
    bUseGravityOverride=true
    simulationSpace=BoneRelative
    alphaBoolBlend {blendOption=Linear}
    boundBone {boneName=dyn_leave_02_r}
    componentAppliedLinearAccClamp=<3000 3000 3000>
    componentLinearAccScale=<1 1 1>
    componentLinearVelScale=<1 1 1>
    gravityOverride {z=-980}
    physicsBodyDefinitions [{
      boundBone {boneName=dyn_leave_02_r}
      boxExtents=<11 11 11>
      constraintSetup {angularLimitsMax=<0 0 5>, angularLimitsMin=<0 0 -80>, angularTarget=<1 0 0>}
      localJointOffset=<0.8 0 0>
     }]
    relativeSpaceBone {boneName=def_umbrella_base}
   }
   n39 = Anim Dynamics - Bone: dyn_stem_low_01(
    ComponentPose=@n38
    Alpha=1
    SimSpaceSettings=(SimSpaceAngularAlpha=0,MaxAngularVelocity=10000,MaxAngularAcceleration=10000,ExternalAngularVelocity=(X=0,Y=0,Z=0))
    ExternalForce="0,0,0"
   ) {
    angularDampingOverride=0.99
    angularSpringConstant=300
    bAngularSpring=true
    bOverrideAngularDamping=true
    bUseGravityOverride=true
    simulationSpace=BoneRelative
    alphaBoolBlend {blendOption=Linear}
    boundBone {boneName=dyn_stem_low_01}
    componentAppliedLinearAccClamp=<3000 3000 3000>
    componentLinearAccScale=<1 1 1>
    componentLinearVelScale=<1 1 1>
    gravityOverride {z=-980}
    physicsBodyDefinitions [{
      boundBone {boneName=dyn_stem_low_01}
      boxExtents=<11 11 11>
      constraintSetup {angularLimitsMax=<0 5 20>, angularLimitsMin=<0 -5 -45>, angularTarget=<1 0 0>}
      localJointOffset=<0.6 0 0>
     }]
    relativeSpaceBone {boneName=def_umbrella_base}
   }
   n40 = Anim Dynamics - Bone: dyn_stem_low_02(
    ComponentPose=@n39
    Alpha=1
    SimSpaceSettings=(SimSpaceAngularAlpha=0,MaxAngularVelocity=10000,MaxAngularAcceleration=10000,ExternalAngularVelocity=(X=0,Y=0,Z=0))
    ExternalForce="0,0,0"
   ) {
    angularDampingOverride=0.99
    angularSpringConstant=300
    bAngularSpring=true
    bOverrideAngularDamping=true
    bUseGravityOverride=true
    simulationSpace=BoneRelative
    alphaBoolBlend {blendOption=Linear}
    boundBone {boneName=dyn_stem_low_02}
    componentAppliedLinearAccClamp=<3000 3000 3000>
    componentLinearAccScale=<1 1 1>
    componentLinearVelScale=<1 1 1>
    gravityOverride {z=-980}
    physicsBodyDefinitions [{
      boundBone {boneName=dyn_stem_low_02}
      boxExtents=<11 11 11>
      constraintSetup {angularLimitsMax=<0 5 30>, angularLimitsMin=<0 -5 -30>, angularTarget=<1 0 0>}
      localJointOffset=<0.6 0 0>
     }]
    relativeSpaceBone {boneName=def_umbrella_base}
   }
   Output Pose(Result=Blend Poses by bool(BlendPose_0=Component To Local(ComponentPose=@n40), BlendPose_1=Component To Local(ComponentPose=@n20)) {blendType=Linear})
   Save cached pose 'InputDefault'(Pose=Input Pose()) {cacheName=InputDefault}
  ]}
 EventGraph {nodes [
   n1 = Event BlueprintInitializeAnimation()
   n2 = Make Anim Phys Sim Space Settings!(SimSpaceAngularAlpha:float=0.3, MaxAngularVelocity:float=1800, MaxAngularAcceleration:float=1800, ExternalAngularVelocity:Vector="0,0,0") -> (AnimPhysSimSpaceSettings:Anim Phys Sim Space Settings Structure) {is=title K2Node_MakeStruct}
   n3 = Make Anim Phys Sim Space Settings(SimSpaceAngularAlpha=0.6, MaxAngularVelocity=1800, MaxAngularAcceleration=1800, ExternalAngularVelocity="0,0,0")
   n4 = Get bIsOnPawn() -> (bIsOnPawn:bool)
   Set Sim Space Settings(
    execute=[@n1.then, @K2Node_Event_2.then]
    Sim Space Settings:Anim Phys Sim Space Settings Structure=Select(
     Option 0:Anim Phys Sim Space Settings Structure=Make Anim Phys Sim Space Settings(SimSpaceAngularAlpha=0.1, MaxAngularVelocity=1500, MaxAngularAcceleration=1500, ExternalAngularVelocity="0,0,0")
     Option 1:Anim Phys Sim Space Settings Structure=Select(Option 0=@n3, Option 1=@n2, Index=@n4)
     Index=Get bIsInFrontEnd()
    ) -> (ReturnValue:Anim Phys Sim Space Settings Structure)
   ) -> (Output_Get:Anim Phys Sim Space Settings Structure)
   K2Node_Event_2 = CustomEvent!() -> (
    OutputDelegate
    then
    CustomBaseMesh:Skeletal Mesh Component Object Reference
    CustomParts:Array of Custom Character Part Object References
    CustomMeshParts:Array of Skeletal Mesh Component Budgeted Object References
   ) {is=title K2Node_CustomEvent}
   TryGetPawnOwner()
   Event BlueprintUpdateAnimation()
  ]}
]
variables [Sim Space Settings {
  type=AnimPhysSimSpaceSettings
  default {maxAngularAcceleration=10000, maxAngularVelocity=10000, simSpaceAngularAlpha=0, externalAngularVelocity=<0 0 0>}
 }]
